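# Sky material function that adds a radial Rift Anomaly crystal layer to an input color, blended in by the RiftAnomalyActive collection parameter.
# A mirrored camera vector is plane-projected and converted to radial UVs for a panning crystal mask, which with sky light direction drives a two-color tint.
# Named reroutes carry a channel-split plasma rainbow and opacity masks, with radial distance fading the result.
# Texture samples use zero derivatives, bypassing automatic mip selection.
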MaterialFunction /Game/TimeOfDay/TODM_Sequence/Materials/Sky/GameplayAssets/MF_Crystal_RA_Sky
graph {nodes [
  Constant() {r=1}
  n1 = FunctionInput()
  n2 = Multiply(
   A=DotProduct(
    A=SkyAtmosphereLightDirection()
    B=MaterialFunctionCall!(Multiplier, Offset) {materialFunction=/Game/TimeOfDay/TODM_Sequence/Materials/Sky/MF_MirroredCameraVector}.Result
   )
   B=Constant() {r=1.1887}
  )
  n3 = MaterialFunctionCall(Multiplier=ScalarParameter() {defaultValue=1, parameterName=CameraMultiplier}) {materialFunction=&MF_MirroredCameraVector}
  n4 = Add(A=ComponentMask(None=@n3.Result) {b=true}) {constB=-0.2}
  n5 = Multiply(A=Divide(B=Multiply(A=@n4) {constB=0.8}), B=@n3.Result)
  n6 = Add(
   A=Multiply(A=ComponentMask(None=@n5) {g=true, r=true}, B=ScalarParameter() {defaultValue=0.8761, parameterName=RA_Crystal_Height})
   B=AppendVector(
    A=ScalarParameter() {defaultValue=0.6907, parameterName=RA_Crystal_OffsetX}
    B=ScalarParameter() {defaultValue=0.5989, parameterName=RA_Crystal_OffsetY}
   )
  )
  n7 = MaterialFunctionCall(Vector or UVs=@n6) {materialFunction=&VectorToRadialValue}
  n8 = Panner(Coordinate=Add(A=@n7.Radial Coordinates, B=ScalarParameter() {defaultValue=1, parameterName=fasdf}) {constB=0}) {speedX=-0.03, speedY=-0.025}
  n9 = Constant()
  n10 = TextureSample!(UVs=@n8, Tex, "DDX(UVs)"=@n9, "DDY(UVs)"=@n9, Apply View MipBias) {mipValueMode=TMVM_Derivative, samplerType=SAMPLERTYPE_Masks, texture=&/CrystalSky_M}
  n11 = LinearInterpolate(A=ScalarParameter() {defaultValue=-0.1443, parameterName=Min}, Alpha=@n10.R) {constA=0.2}
  n12 = LinearInterpolate(
   A=VectorParameter() {parameterName=RACrystal_ColorA, defaultValue=<rgba 0.0601 0.0151 0.0625 1>}.RGB
   B=VectorParameter() {parameterName=RACrystal_ColorB, defaultValue=<rgba 0.1999 0.0984 0.2708 1>}.RGB
   Alpha=Saturate(None=Add(A=@n2, B=@n11))
  )
  n13 = NamedRerouteUsage()
  n14 = Saturate(None=NamedRerouteUsage())
  n15 = Multiply(A=NamedRerouteUsage(), B=ScalarParameter() {defaultValue=2.8678, parameterName=OverallOpacity})
  n16 = Multiply(A=@n7.Linear Distance, B=ScalarParameter() {defaultValue=3.1356, parameterName=skew}) {constB=2}
  n17 = ScalarParameter() {defaultValue=0.9364, parameterName=PowerSkew}
  n18 = Multiply(
   A=Multiply(A=LinearInterpolate(A=@n12, B=@n13, Alpha=@n14), B=Saturate(None=@n15))
   B=Saturate(None=Power(Base=@n16, Exp=@n17) {constExponent=4})
  )
  n19 = Add(A=@n1, B=@n18)
  n20 = CollectionParameter() {collection=&FortniteMaterialParameters, parameterName=RiftAnomalyActive}
  FunctionOutput(None=LinearInterpolate(A=@n1, B=@n19, Alpha=@n20))
  n21 = VectorParameter() {parameterName=Speed, defaultValue=<rgba 0 -0.2 0 0>}
  n22 = Multiply(
   A=Panner(
    Coordinate=MaterialFunctionCall(Vector or UVs=@n6) {materialFunction=&VectorToRadialValue}.Radial Coordinates
    Time=Time()
    Speed=ComponentMask(None=@n21.RGB) {g=true, r=true}
   ) {speedY=-0.4}
   B=Constant2Vector() {g=1, r=1}
  )
  n23 = Add(
   A=Multiply(A=NamedRerouteUsage(), B=ScalarParameter() {defaultValue=0.1296, parameterName=gsdfgsdfg})
   B=NamedRerouteDeclaration(None=Add(A=@n22)) {name=Radial UVs, nodeColor=<rgba 0.8047 1 0 1>}
  )
  n24 = ScalarParameter() {defaultValue=0.02, parameterName=Color offset}
  n25 = Constant()
  n26 = TextureSampleParameter2D!(UVs=Add(A=@n24, B=@n23), "DDX(UVs)"=@n25, "DDY(UVs)"=@n25, Apply View MipBias) {
   mipValueMode=TMVM_Derivative
   parameterName=Tex Noise
   samplerType=SAMPLERTYPE_LinearGrayscale
   texture=/Game/Effects/Fort_Effects/Emotes/UltraLight/Texture/T_PlasmaCoils
  }
  n27 = Constant()
  n28 = TextureSampleParameter2D(UVs=@n23, "DDX(UVs)"=@n27, "DDY(UVs)"=@n27) {
   mipValueMode=TMVM_Derivative
   parameterName=Tex Noise
   samplerType=SAMPLERTYPE_LinearGrayscale
   texture=&T_PlasmaCoils
  }
  n29 = Constant()
  n30 = TextureSampleParameter2D(UVs=Subtract(A=@n23, B=@n24), "DDX(UVs)"=@n29, "DDY(UVs)"=@n29) {
   mipValueMode=TMVM_Derivative
   parameterName=Tex Noise
   samplerType=SAMPLERTYPE_LinearGrayscale
   texture=&T_PlasmaCoils
  }
  n31 = MaterialFunctionCall(X=@n26.R, Y=@n28.R, Z=@n30.R) {materialFunction=&MakeFloat3}
  NamedRerouteDeclaration(None=Saturate(None=Power(Base=@n31.Result) {constExponent=3})) {name=Rainbow, nodeColor=<rgba 0.0781 0 1 1>}
  n32 = Constant()
  n33 = TextureSampleParameter2D("DDX(UVs)"=@n32, "DDY(UVs)"=@n32) {
   mipValueMode=TMVM_Derivative
   parameterName=Tex Base
   samplerType=SAMPLERTYPE_LinearGrayscale
   texture=/Game/Characters/Player/Male/Medium/Bodies/M_MED_Citadel/FX/T_Citadel_Noise_Dissolve_Handpaint
  }
  n34 = ScalarParameter() {defaultValue=0.7, parameterName=Center Alpha}
  n35 = Add(A=Add(B=Multiply(A=@n33.R, B=@n34) {constB=0.7})) {constB=0}
  n36 = SphereMask(A=TextureCoordinate(), B=ScalarParameter() {defaultValue=0.5, parameterName=Opacity sphere}) {attenuationRadius=0.5, hardnessPercent=0}
  n37 = OneMinus(None=@n36)
  n38 = LinearInterpolate(Alpha=@n37) {constA=-0.1, constB=2}
  NamedRerouteDeclaration(None=Power(Base=Power(Base=@n35, Exp=@n38)) {constExponent=1.2}) {name=OpacityMask, nodeColor=<rgba 0.2422 1 0 1>}
  n39 = Multiply(A=@n36, B=@n37)
  n40 = ScalarParameter() {defaultValue=1.6042, parameterName=Opacitypower}
  NamedRerouteDeclaration(None=Multiply(A=Power(Base=@n39, Exp=@n40) {constExponent=1.5}, B=ScalarParameter() {defaultValue=1.36, parameterName=Balance})) {name=OpacitySphere, nodeColor=<rgba 0 0.4844 1 1>}
  n41 = Constant()
  n42 = TextureSample(UVs=@n8, "DDX(UVs)"=@n41, "DDY(UVs)"=@n41) {mipValueMode=TMVM_Derivative, samplerType=SAMPLERTYPE_Masks, texture=&CrystalSky_M}
  NamedRerouteDeclaration(None=@n42.B) {name=Crystal Mask, nodeColor=<rgba 0.5 1 0 1>}
  NamedRerouteUsage()
  OneMinus(None=@n7.Linear Distance)
  Reroute(None=@n19)
  Saturate(None=@n18)
  Sine(None=@n36)
 ]}
